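# Physics asset for the F_MED_Taco female medium player body cosmetic, covering dynamic boot straps and shoelaces on the lower legs.
# Capsules on both calves link to a small pelvis capsule with free angular motion.
# Box bodies form two three-link shoelace chains, both anchored to the right calf, and two boot straps, each constrained directly to the left calf.
# Tight swing and twist limits restrict lace movement, while straps swing mainly on one axis.
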PhysicsAsset /BRCosmetics/Characters/Player/Female/Medium/Bodies/F_MED_Taco/Meshes/F_MED_Taco_Physics
bodies {
 calf_l [{shapeType=Capsule, length=13.5379, radius=6.0008, center=<-23.1544 -3.5655 -0.4142>, rotation=<rot 86.3857 -3.415 -3.6234>}]
 calf_r [{shapeType=Capsule, length=32.6476, radius=5.2075, center=<20.8321 4.5844 0.2298>, rotation=<rot 89.4304 176.6128 174.647>}]
 dyn_bootstrap_01 [{shapeType=Box, extentX=4.0365, extentY=2, extentZ=0.5, center=<2.6966 0 0>}]
 dyn_bootstrap_02 [{shapeType=Box, extentX=4.0365, extentY=2, extentZ=0.5, center=<2.5449 0.2312 -0.3155>, rotation=<rot -6.4452 1.5554 -0.1746>}]
 dyn_shoelace_base_l [{shapeType=Box, extentX=4.8686, extentY=0.7389, extentZ=3.1276, center=<2.8248 0.0481 0>}]
 dyn_shoelace_base_r [{shapeType=Box, extentX=4.6457, extentY=1.033, extentZ=2.6827, center=<2.6195 -0.0001 0>}]
 dyn_shoelace_l_01 [{shapeType=Box, extentX=4.7793, extentY=1.0461, extentZ=2.9689, center=<2.6674 -0.0941 0>, rotation=<rot 0 -1.6043 0>}]
 dyn_shoelace_l_02 [{shapeType=Box, extentX=2.9225, extentY=1.3592, extentZ=1.2486, center=<0.124 0 -0.5776>, rotation=<rot 12.1211 0 0>}]
 dyn_shoelace_r_01 [{shapeType=Box, extentX=5.1571, extentY=1.0636, extentZ=3.7719, center=<2.6775 -0.2275 0>}]
 dyn_shoelace_r_02 [{shapeType=Box, extentX=3.2549, extentY=1.0382, extentZ=2, center=<0 0 -0.7877>}]
 pelvis [{shapeType=Capsule, length=2, radius=2}]
}
constraints [
 {bone1Name=dyn_shoelace_l_02, bone2Name=dyn_shoelace_l_01, swing1LimitDegrees=10, swing2LimitDegrees=10, twistLimitDegrees=10}
 {bone1Name=dyn_shoelace_base_l, bone2Name=calf_r, swing1LimitDegrees=10, swing2LimitDegrees=10, twistLimitDegrees=10}
 {bone1Name=dyn_shoelace_l_01, bone2Name=dyn_shoelace_base_l, swing1LimitDegrees=10, swing2LimitDegrees=10, twistLimitDegrees=10}
 {bone1Name=dyn_shoelace_base_r, bone2Name=calf_r, swing1LimitDegrees=15, swing2LimitDegrees=15, twistLimitDegrees=15}
 {bone1Name=dyn_shoelace_r_01, bone2Name=dyn_shoelace_base_r, swing1LimitDegrees=15, swing2LimitDegrees=10, twistLimitDegrees=15}
 {bone1Name=dyn_shoelace_r_02, bone2Name=dyn_shoelace_r_01, swing1LimitDegrees=15, swing2LimitDegrees=10, twistLimitDegrees=15}
 {bone1Name=dyn_bootstrap_02, bone2Name=calf_l, swing1LimitDegrees=2, swing2LimitDegrees=20, twistLimitDegrees=2}
 {bone1Name=dyn_bootstrap_01, bone2Name=calf_l, swing1LimitDegrees=2, swing2LimitDegrees=20, twistLimitDegrees=2}
 {
  bone1Name=calf_l
  bone2Name=pelvis
  swing1LimitDegrees=45
  swing1Motion=Free
  swing2LimitDegrees=45
  swing2Motion=Free
  twistLimitDegrees=45
  twistMotion=Free
 }
 {
  bone1Name=calf_r
  bone2Name=pelvis
  swing1LimitDegrees=45
  swing1Motion=Free
  swing2LimitDegrees=45
  swing2Motion=Free
  twistLimitDegrees=45
  twistMotion=Free
 }
]
